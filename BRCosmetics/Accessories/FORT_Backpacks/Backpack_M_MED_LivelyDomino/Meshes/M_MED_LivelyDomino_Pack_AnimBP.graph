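# Animation Blueprint for the LivelyDomino back bling, an animated pet with idle, fidget and elimination-reaction animations.
# A nested state machine layers timed random tail/ear additive twitches and idle breaks, switching to Elim on enemy eliminations or a delayed frontend trigger.
# Its pose passes through a pack offset and AnimDynamics on katana, dice and tail bones (band dynamics zeroed).
# Events toggle a material hide parameter according to matching pickaxe and skydiving state.
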AnimBlueprint /BRCosmetics/Accessories/FORT_Backpacks/Backpack_M_MED_LivelyDomino/Meshes/M_MED_LivelyDomino_Pack_AnimBP
target_skeleton=&/M_MED_LivelyDomino_Pack_Skeleton
graphs [
 AnimGraph {nodes [
   n1 = Anim Dynamics - Bone: tail_c_01(ComponentPose=Local To Component(LocalPose=Use cached pose 'HeadCache'()), Alpha=0.5, ExternalForce="0,0,0") {
    angularDampingOverride=1
    angularSpringConstant=2000
    bAngularSpring=true
    bOverrideAngularDamping=true
    bUseGravityOverride=true
    gravityScale=0.35
    simulationSpace=RootRelative
    alphaBoolBlend {blendOption=Linear}
    boundBone {boneName=tail_c_01}
    componentLinearAccScale=<1 1 1>
    componentLinearVelScale=<1 1 1>
    gravityOverride {z=-2000}
    physicsBodyDefinitions [{
      boundBone {boneName=tail_c_01}
      boxExtents=<20 30 30>
      constraintSetup {angularLimitsMax=<5 5 5>, angularLimitsMin=<-5 -5 -5>, angularTarget=<1 0 0>}
      localJointOffset=<2 0 0>
     }]
   }
   n2 = Anim Dynamics - Bone: tail_c_02(ComponentPose=@n1, Alpha=0.5, ExternalForce="0,0,0") {
    angularDampingOverride=1
    angularSpringConstant=2000
    bAngularSpring=true
    bOverrideAngularDamping=true
    bUseGravityOverride=true
    gravityScale=0.35
    simulationSpace=RootRelative
    alphaBoolBlend {blendOption=Linear}
    boundBone {boneName=tail_c_02}
    componentLinearAccScale=<1 1 1>
    componentLinearVelScale=<1 1 1>
    gravityOverride {z=-2000}
    physicsBodyDefinitions [{
      boundBone {boneName=tail_c_02}
      boxExtents=<20 30 30>
      constraintSetup {angularLimitsMax=<20 20 20>, angularLimitsMin=<-20 -20 -20>, angularTarget=<1 0 0>}
      localJointOffset=<3 0 0>
     }]
   }
   n3 = Anim Dynamics - Bone: tail_c_03(ComponentPose=@n2, Alpha=0.5, ExternalForce="0,0,0") {
    angularDampingOverride=1
    angularSpringConstant=2000
    bAngularSpring=true
    bOverrideAngularDamping=true
    bUseGravityOverride=true
    gravityScale=0.35
    simulationSpace=RootRelative
    alphaBoolBlend {blendOption=Linear}
    boundBone {boneName=tail_c_03}
    componentLinearAccScale=<1 1 1>
    componentLinearVelScale=<1 1 1>
    gravityOverride {z=-2000}
    physicsBodyDefinitions [{
      boundBone {boneName=tail_c_03}
      boxExtents=<20 30 30>
      constraintSetup {angularLimitsMax=<25 25 25>, angularLimitsMin=<-25 -25 -25>, angularTarget=<1 0 0>}
      localJointOffset=<6 0 0>
     }]
   }
   n4 = Anim Dynamics - Bone: tail_c_04(ComponentPose=@n3, Alpha=0.5, ExternalForce="0,0,0") {
    angularDampingOverride=1
    angularSpringConstant=2000
    bAngularSpring=true
    bOverrideAngularDamping=true
    bUseGravityOverride=true
    gravityScale=0.35
    simulationSpace=RootRelative
    alphaBoolBlend {blendOption=Linear}
    boundBone {boneName=tail_c_04}
    componentLinearAccScale=<1 1 1>
    componentLinearVelScale=<1 1 1>
    gravityOverride {z=-2000}
    physicsBodyDefinitions [{
      boundBone {boneName=tail_c_04}
      boxExtents=<20 30 30>
      constraintSetup {angularLimitsMax=<25 25 25>, angularLimitsMin=<-25 -25 -25>, angularTarget=<1 0 0>}
      localJointOffset=<6 0 0>
     }]
   }
   Output Pose(Result=Component To Local(ComponentPose=@n4)) {layerGroup=None}
   AnimGraphNode_StateMachine_0 = Animation_Idle() {is=title AnimGraphNode_StateMachine}
   n5 = "Transform (Modify) Bone - Bone: pack"!(ComponentPose=Local To Component(LocalPose=@AnimGraphNode_StateMachine_0), Alpha, Translation="0,-1,6") -> (Pose) {
    translationMode=BMM_Additive
    translationSpace=BCS_BoneSpace
    alphaBoolBlend {blendOption=Linear}
    boneToModify {boneName=pack}
    translation {y=-1, z=6}
   }
   n6 = Anim Dynamics - Bone: katana_c(ComponentPose=@n5, Alpha=1, ExternalForce="0,0,0") {
    angularDampingOverride=1
    angularSpringConstant=20000
    bAngularSpring=true
    bOverrideAngularDamping=true
    bUseGravityOverride=true
    gravityScale=0.35
    simulationSpace=RootRelative
    alphaBoolBlend {blendOption=Linear}
    boundBone {boneName=katana_c}
    componentLinearAccScale=<1 1 1>
    componentLinearVelScale=<1 1 1>
    gravityOverride {z=-980}
    physicsBodyDefinitions [{
      boundBone {boneName=katana_c}
      boxExtents=<20 30 30>
      constraintSetup {angularLimitsMax=<0 15 0>, angularLimitsMin=<0 -10 0>, angularTarget=<1 0 0>}
      localJointOffset=<1.5 0 0>
     }]
   }
   n7 = Anim Dynamics - Bone: dice_a_01(ComponentPose=@n6, Alpha=1, ExternalForce="0,0,0") {
    angularDampingOverride=1
    angularSpringConstant=10000
    bAngularSpring=true
    bOverrideAngularDamping=true
    bUseGravityOverride=true
    gravityScale=0.35
    simulationSpace=RootRelative
    alphaBoolBlend {blendOption=Linear}
    boundBone {boneName=dice_a_01}
    componentLinearAccScale=<1 1 1>
    componentLinearVelScale=<1 1 1>
    gravityOverride {z=-1000}
    physicsBodyDefinitions [{
      boundBone {boneName=dice_a_01}
      boxExtents=<20 50 30>
      constraintSetup {angularLimitsMax=<5 10 20>, angularLimitsMin=<-5 -5 -60>, angularTarget=<1 0 0>}
      localJointOffset=<5 0 0>
     }]
   }
   n8 = Anim Dynamics - Bone: dice_a_02(ComponentPose=@n7, Alpha=1, ExternalForce="0,0,0") {
    angularDampingOverride=1
    angularSpringConstant=20000
    bAngularSpring=true
    bOverrideAngularDamping=true
    bUseGravityOverride=true
    gravityScale=0.35
    simulationSpace=RootRelative
    alphaBoolBlend {blendOption=Linear}
    boundBone {boneName=dice_a_02}
    componentLinearAccScale=<1 1 1>
    componentLinearVelScale=<1 1 1>
    gravityOverride {z=-900}
    physicsBodyDefinitions [{
      boundBone {boneName=dice_a_02}
      boxExtents=<20 40 30>
      constraintSetup {angularLimitsMax=<20 40 20>, angularLimitsMin=<-20 -20 -40>, angularTarget=<1 0 0>}
      localJointOffset=<3 0 0>
     }]
   }
   n9 = Anim Dynamics - Bone: dice_b_01(ComponentPose=@n8, Alpha=1, ExternalForce="0,0,0") {
    angularDampingOverride=1
    angularSpringConstant=10000
    bAngularSpring=true
    bOverrideAngularDamping=true
    bUseGravityOverride=true
    gravityScale=0.35
    simulationSpace=RootRelative
    alphaBoolBlend {blendOption=Linear}
    boundBone {boneName=dice_b_01}
    componentLinearAccScale=<1 1 1>
    componentLinearVelScale=<1 1 1>
    gravityOverride {z=-1000}
    physicsBodyDefinitions [{
      boundBone {boneName=dice_b_01}
      boxExtents=<20 50 30>
      constraintSetup {angularLimitsMax=<5 10 20>, angularLimitsMin=<-5 -5 -60>, angularTarget=<1 0 0>}
      localJointOffset=<5 0 0>
     }]
   }
   n10 = Anim Dynamics - Bone: dice_b_02(ComponentPose=@n9, Alpha=1, ExternalForce="0,0,0") {
    angularDampingOverride=1
    angularSpringConstant=20000
    bAngularSpring=true
    bOverrideAngularDamping=true
    bUseGravityOverride=true
    gravityScale=0.35
    simulationSpace=RootRelative
    alphaBoolBlend {blendOption=Linear}
    boundBone {boneName=dice_b_02}
    componentLinearAccScale=<1 1 1>
    componentLinearVelScale=<1 1 1>
    gravityOverride {z=-900}
    physicsBodyDefinitions [{
      boundBone {boneName=dice_b_02}
      boxExtents=<20 40 30>
      constraintSetup {angularLimitsMax=<20 40 20>, angularLimitsMin=<-20 -20 -40>, angularTarget=<1 0 0>}
      localJointOffset=<3 0 0>
     }]
   }
   Save cached pose 'KatanaCache'(Pose=Component To Local(ComponentPose=@n10)) {cacheName=KatanaCache}
   n11 = Anim Dynamics - Bone: band_l(ComponentPose=Local To Component(LocalPose=Use cached pose 'KatanaCache'()), Alpha=0, ExternalForce="0,0,0") {
    angularDampingOverride=1
    angularSpringConstant=20000
    bAngularSpring=true
    bOverrideAngularDamping=true
    bUseGravityOverride=true
    gravityScale=0.35
    simulationSpace=RootRelative
    alphaBoolBlend {blendOption=Linear}
    boundBone {boneName=band_l}
    componentLinearAccScale=<1 1 1>
    componentLinearVelScale=<1 1 1>
    gravityOverride {z=-980}
    physicsBodyDefinitions [{
      boundBone {boneName=band_l}
      boxExtents=<20 30 30>
      constraintSetup {angularLimitsMax=<0 15 15>, angularLimitsMin=<0 -15 0>, angularTarget=<1 0 0>}
      localJointOffset=<4 0 0>
     }]
   }
   n12 = Anim Dynamics - Bone: band_r(ComponentPose=@n11, Alpha=0, ExternalForce="0,0,0") {
    angularDampingOverride=1
    angularSpringConstant=20000
    bAngularSpring=true
    bOverrideAngularDamping=true
    bUseGravityOverride=true
    gravityScale=0.35
    simulationSpace=RootRelative
    alphaBoolBlend {blendOption=Linear}
    boundBone {boneName=band_r}
    componentLinearAccScale=<1 1 1>
    componentLinearVelScale=<1 1 1>
    gravityOverride {z=-980}
    physicsBodyDefinitions [{
      boundBone {boneName=band_r}
      boxExtents=<20 30 30>
      constraintSetup {angularLimitsMax=<0 15 15>, angularLimitsMin=<0 -15 0>, angularTarget=<1 0 0>}
      localJointOffset=<4 0 0>
     }]
   }
   Save cached pose 'HeadCache'(Pose=Component To Local(ComponentPose=@n12)) {cacheName=HeadCache}
  ]}
 Animation_Idle {
  entry=Idle
  path=AnimGraph.AnimGraphNode_StateMachine_0.Animation_Idle
  ids=id
  nodes [
   AnimStateNode_0 = Idle()
   AnimStateNode_1 = Elim() {
    is=title AnimStateNode
    stateEntered {guid=#1, notifyName=EnteredElim}
    stateLeft {guid=#2, notifyName=ExitedElim}
   }
   AnimStateTransitionNode_0 = AnimStateTransitionNode()
   AnimStateTransitionNode_1 = AnimStateTransitionNode() {bAutomaticRuleBasedOnSequencePlayerInState=true}
  ]
  links [
   {
    dst=Idle
    rule=AnimGraph.AnimGraphNode_StateMachine_0.Animation_Idle.AnimStateTransitionNode_1.Transition
    src=Elim
    via=AnimStateTransitionNode_1
   }
   {
    dst=Elim
    rule=AnimGraph.AnimGraphNode_StateMachine_0.Animation_Idle.AnimStateTransitionNode_0.Transition
    src=Idle
    via=AnimStateTransitionNode_0
   }
  ]
 }
 Idle {
  path=AnimGraph.AnimGraphNode_StateMachine_0.Animation_Idle.AnimStateNode_0.Idle
  nodes [
   AnimGraphNode_StateMachine_0 = Idle() {is=title AnimGraphNode_StateMachine}
   Output Animation Pose(Result=@AnimGraphNode_StateMachine_0)
  ]
 }
 Idle {
  entry=Normal Idle
  path=AnimGraph.AnimGraphNode_StateMachine_0.Animation_Idle.AnimStateNode_0.Idle.AnimGraphNode_StateMachine_0.Idle
  ids=id
  nodes [
   AnimStateNode_0 = Normal Idle() {is=title AnimStateNode, stateEntered {guid=#3, notifyName=Set Random Interger}}
   AnimStateNode_1 = Idle Break() {
    is=title AnimStateNode
    stateEntered {guid=#4, notifyName=Entered Idle Break}
    stateLeft {guid=#5, notifyName=Exited Idle Break}
   }
   AnimStateTransitionNode_0 = AnimStateTransitionNode()
   AnimStateTransitionNode_1 = AnimStateTransitionNode() {bAutomaticRuleBasedOnSequencePlayerInState=true}
  ]
  links [
   {
    dst=Normal Idle
    rule=AnimGraph.AnimGraphNode_StateMachine_0.Animation_Idle.AnimStateNode_0.Idle.AnimGraphNode_StateMachine_0.Idle.AnimStateTransitionNode_1.Transition
    src=Idle Break
    via=AnimStateTransitionNode_1
   }
   {
    dst=Idle Break
    rule=AnimGraph.AnimGraphNode_StateMachine_0.Animation_Idle.AnimStateNode_0.Idle.AnimGraphNode_StateMachine_0.Idle.AnimStateTransitionNode_0.Transition
    src=Normal Idle
    via=AnimStateTransitionNode_0
   }
  ]
 }
 Normal Idle {
  path=AnimGraph.AnimGraphNode_StateMachine_0.Animation_Idle.AnimStateNode_0.Idle.AnimGraphNode_StateMachine_0.Idle.AnimStateNode_0.Normal Idle
  nodes [
   n1 = Sequence Player 'LivelyDomino_idle_M_MED'() {sequence=/BRCosmetics/Accessories/FORT_Backpacks/Backpack_M_MED_LivelyDomino/Animation/LivelyDomino_idle_M_MED}
   AnimGraphNode_StateMachine_1 = TailTwitch() {is=title AnimGraphNode_StateMachine}
   AnimGraphNode_StateMachine_2 = EarTwitch() {is=title AnimGraphNode_StateMachine}
   Output Animation Pose(Result=Apply Additive(
     Base=Apply Additive(Base=@n1, Additive=@AnimGraphNode_StateMachine_1) {alphaBoolBlend {blendOption=Linear}}
     Additive=@AnimGraphNode_StateMachine_2
    ) {alphaBoolBlend {blendOption=Linear}})
  ]
 }
 TailTwitch {
  entry=Identity
  path=AnimGraph.AnimGraphNode_StateMachine_0.Animation_Idle.AnimStateNode_0.Idle.AnimGraphNode_StateMachine_0.Idle.AnimStateNode_0.Normal Idle.AnimGraphNode_StateMachine_1.TailTwitch
  ids=id
  nodes [
   AnimStateNode_0 = Identity() {is=title AnimStateNode}
   AnimStateNode_1 = Twitch() {
    is=title AnimStateNode
    stateEntered {guid=#6, notifyName=Entered Tail Twitch}
    stateLeft {guid=#7, notifyName=Exited Tail Twitch}
   }
   AnimStateTransitionNode_0 = AnimStateTransitionNode()
   AnimStateTransitionNode_1 = AnimStateTransitionNode() {bAutomaticRuleBasedOnSequencePlayerInState=true}
  ]
  links [
   {
    dst=Twitch
    rule=AnimGraph.AnimGraphNode_StateMachine_0.Animation_Idle.AnimStateNode_0.Idle.AnimGraphNode_StateMachine_0.Idle.AnimStateNode_0.Normal Idle.AnimGraphNode_StateMachine_1.TailTwitch.AnimStateTransitionNode_0.Transition
    src=Identity
    via=AnimStateTransitionNode_0
   }
   {
    dst=Identity
    rule=AnimGraph.AnimGraphNode_StateMachine_0.Animation_Idle.AnimStateNode_0.Idle.AnimGraphNode_StateMachine_0.Idle.AnimStateNode_0.Normal Idle.AnimGraphNode_StateMachine_1.TailTwitch.AnimStateTransitionNode_1.Transition
    src=Twitch
    via=AnimStateTransitionNode_1
   }
  ]
 }
 Identity {
  path=AnimGraph.AnimGraphNode_StateMachine_0.Animation_Idle.AnimStateNode_0.Idle.AnimGraphNode_StateMachine_0.Idle.AnimStateNode_0.Normal Idle.AnimGraphNode_StateMachine_1.TailTwitch.AnimStateNode_0.Identity
  nodes [Output Animation Pose(Result=Additive Identity Pose())]
 }
 Twitch {
  path=AnimGraph.AnimGraphNode_StateMachine_0.Animation_Idle.AnimStateNode_0.Idle.AnimGraphNode_StateMachine_0.Idle.AnimStateNode_0.Normal Idle.AnimGraphNode_StateMachine_1.TailTwitch.AnimStateNode_1.Twitch
  nodes [
   n1 = "Sequence Player (additive) 'LivelyDomino_idlebreaks_Tails_M_MED'"() {
    bLoopAnimation=false
    sequence=/BRCosmetics/Accessories/FORT_Backpacks/Backpack_M_MED_LivelyDomino/Animation/LivelyDomino_idlebreaks_Tails_M_MED
   }
   Output Animation Pose(Result=@n1)
  ]
 }
 Transition {
  path=AnimGraph.AnimGraphNode_StateMachine_0.Animation_Idle.AnimStateNode_0.Idle.AnimGraphNode_StateMachine_0.Idle.AnimStateNode_0.Normal Idle.AnimGraphNode_StateMachine_1.TailTwitch.AnimStateTransitionNode_0.Transition
  nodes [Result(bCanEnterTransition=Get Tail Twitch() -> (Tail Twitch:bool))]
 }
 Transition {
  path=AnimGraph.AnimGraphNode_StateMachine_0.Animation_Idle.AnimStateNode_0.Idle.AnimGraphNode_StateMachine_0.Idle.AnimStateNode_0.Normal Idle.AnimGraphNode_StateMachine_1.TailTwitch.AnimStateTransitionNode_1.Transition
  nodes [Result()]
 }
 EarTwitch {
  entry=Identity
  path=AnimGraph.AnimGraphNode_StateMachine_0.Animation_Idle.AnimStateNode_0.Idle.AnimGraphNode_StateMachine_0.Idle.AnimStateNode_0.Normal Idle.AnimGraphNode_StateMachine_2.EarTwitch
  ids=id
  nodes [
   AnimStateNode_0 = Identity()
   AnimStateNode_1 = EarTwitch() {
    is=title AnimStateNode
    stateEntered {guid=#8, notifyName=Entered Ear Twitch}
    stateLeft {guid=#9, notifyName=Exited Ear Twitch}
   }
   AnimStateTransitionNode_2 = AnimStateTransitionNode()
   AnimStateTransitionNode_3 = AnimStateTransitionNode() {bAutomaticRuleBasedOnSequencePlayerInState=true}
  ]
  links [
   {
    dst=Identity
    rule=AnimGraph.AnimGraphNode_StateMachine_0.Animation_Idle.AnimStateNode_0.Idle.AnimGraphNode_StateMachine_0.Idle.AnimStateNode_0.Normal Idle.AnimGraphNode_StateMachine_2.EarTwitch.AnimStateTransitionNode_3.Transition
    src=EarTwitch
    via=AnimStateTransitionNode_3
   }
   {
    dst=EarTwitch
    rule=AnimGraph.AnimGraphNode_StateMachine_0.Animation_Idle.AnimStateNode_0.Idle.AnimGraphNode_StateMachine_0.Idle.AnimStateNode_0.Normal Idle.AnimGraphNode_StateMachine_2.EarTwitch.AnimStateTransitionNode_2.Transition
    src=Identity
    via=AnimStateTransitionNode_2
   }
  ]
 }
 Identity {
  path=AnimGraph.AnimGraphNode_StateMachine_0.Animation_Idle.AnimStateNode_0.Idle.AnimGraphNode_StateMachine_0.Idle.AnimStateNode_0.Normal Idle.AnimGraphNode_StateMachine_2.EarTwitch.AnimStateNode_0.Identity
  nodes [Output Animation Pose(Result=Additive Identity Pose())]
 }
 EarTwitch {
  path=AnimGraph.AnimGraphNode_StateMachine_0.Animation_Idle.AnimStateNode_0.Idle.AnimGraphNode_StateMachine_0.Idle.AnimStateNode_0.Normal Idle.AnimGraphNode_StateMachine_2.EarTwitch.AnimStateNode_1.EarTwitch
  nodes [
   n1 = "Sequence Player (additive) 'LivelyDomino_idlebreaks_ears_M_MED'"() {
    bLoopAnimation=false
    sequence=/BRCosmetics/Accessories/FORT_Backpacks/Backpack_M_MED_LivelyDomino/Animation/LivelyDomino_idlebreaks_ears_M_MED
   }
   Output Animation Pose(Result=@n1)
  ]
 }
 Transition {
  path=AnimGraph.AnimGraphNode_StateMachine_0.Animation_Idle.AnimStateNode_0.Idle.AnimGraphNode_StateMachine_0.Idle.AnimStateNode_0.Normal Idle.AnimGraphNode_StateMachine_2.EarTwitch.AnimStateTransitionNode_2.Transition
  nodes [Result(bCanEnterTransition=Get Ear Wiggle() -> (Ear Wiggle:bool))]
 }
 Transition {
  path=AnimGraph.AnimGraphNode_StateMachine_0.Animation_Idle.AnimStateNode_0.Idle.AnimGraphNode_StateMachine_0.Idle.AnimStateNode_0.Normal Idle.AnimGraphNode_StateMachine_2.EarTwitch.AnimStateTransitionNode_3.Transition
  nodes [Result()]
 }
 Idle Break {
  path=AnimGraph.AnimGraphNode_StateMachine_0.Animation_Idle.AnimStateNode_0.Idle.AnimGraphNode_StateMachine_0.Idle.AnimStateNode_1.Idle Break
  nodes [
   n1 = Sequence Player 'LivelyDomino_lickpaw_M_MED'() {bLoopAnimation=false, sequence=/BRCosmetics/Accessories/FORT_Backpacks/Backpack_M_MED_LivelyDomino/Animation/LivelyDomino_lickpaw_M_MED}
   n2 = Sequence Player 'LivelyDomino_lookaround_M_MED'() {bLoopAnimation=false, sequence=/BRCosmetics/Accessories/FORT_Backpacks/Backpack_M_MED_LivelyDomino/Animation/LivelyDomino_lookaround_M_MED}
   n3 = Sequence Player 'LivelyDomino_lickrubface_M_MED'() {bLoopAnimation=false, sequence=/BRCosmetics/Accessories/FORT_Backpacks/Backpack_M_MED_LivelyDomino/Animation/LivelyDomino_lickrubface_M_MED}
   n4 = Blend Poses by int!(
    BlendPose_0:Pose Link Structure=@n3
    BlendPose_1:Pose Link Structure=@n2
    BlendPose_2:Pose Link Structure=@n1
    BlendTime_0:float=0
    BlendTime_1:float=0
    BlendTime_2:float=0
    ActiveChildIndex:int=Get RandomInt() -> (RandomInt:int=0)
   ) -> (Pose:Pose Link Structure) {
    is=title AnimGraphNode_BlendListByInt
    blendType=Linear
    blendPose [{linkId=-1, sourceLinkId=-1}, {linkId=-1, sourceLinkId=-1}, {linkId=-1, sourceLinkId=-1}]
    blendTime [0.1,0.1,0.1]
   }
   Output Animation Pose(Result=@n4)
  ]
 }
 Transition {
  path=AnimGraph.AnimGraphNode_StateMachine_0.Animation_Idle.AnimStateNode_0.Idle.AnimGraphNode_StateMachine_0.Idle.AnimStateTransitionNode_0.Transition
  nodes [Result(bCanEnterTransition=Get IdleBreak() -> (IdleBreak:bool))]
 }
 Transition {
  path=AnimGraph.AnimGraphNode_StateMachine_0.Animation_Idle.AnimStateNode_0.Idle.AnimGraphNode_StateMachine_0.Idle.AnimStateTransitionNode_1.Transition
  nodes [Result()]
 }
 Elim {
  path=AnimGraph.AnimGraphNode_StateMachine_0.Animation_Idle.AnimStateNode_1.Elim
  nodes [
   n1 = Sequence Player 'LivelyDomino_Reactive_M_MED'() {bLoopAnimation=false, sequence=/BRCosmetics/Accessories/FORT_Backpacks/Backpack_M_MED_LivelyDomino/Animation/LivelyDomino_Reactive_M_MED}
   Output Animation Pose(Result=@n1)
  ]
 }
 Transition {
  path=AnimGraph.AnimGraphNode_StateMachine_0.Animation_Idle.AnimStateTransitionNode_0.Transition
  nodes [Result(bCanEnterTransition=Get Active Kill() -> (Active Kill:bool))]
 }
 Transition {path=AnimGraph.AnimGraphNode_StateMachine_0.Animation_Idle.AnimStateTransitionNode_1.Transition, nodes [Result()]}
 EventGraph {nodes [
   n1 = Event BlueprintProcessAnimInputEvent()
   > Branch(Condition=IsNamed(self=@n1.AnimInputEvent, InEventName=aie.EnemyPlayerEliminated))
   > Set Active Kill(Active Kill:bool=true) -> (Output_Get:bool)
   AnimNotify_Entered Tail Twitch() {is=title K2Node_Event}
   > Set Tail Twitch(Tail Twitch:bool=false) -> (Output_Get:bool)
   AnimNotify_Entered Ear Twitch() {is=title K2Node_Event}
   > Set Ear Wiggle(Ear Wiggle:bool=false) -> (Output_Get:bool)
   n2 = AnimNotify_Exited Tail Twitch() {is=title K2Node_Event}
   Delay(execute=[@n2.then, @n4.then_1], Duration=RandomFloatInRange(Min=5, Max=10))
   > Set Tail Twitch(Tail Twitch=true)
   n3 = AnimNotify_Exited Ear Twitch() {is=title K2Node_Event}
   Delay(execute=[@n3.then, @n4.then_2], Duration=RandomFloatInRange(Min=5, Max=10))
   > Set Ear Wiggle(Ear Wiggle=true)
   AnimNotify_Entered Idle Break() {is=title K2Node_Event}
   > Set IdleBreak(IdleBreak:bool=false) -> (Output_Get:bool)
   AnimNotify_Exited Idle Break() {is=title K2Node_Event}
   > n4 = Sequence!(execute) -> (then_0, then_1, then_2:Exec)
   > Delay(Duration=RandomFloatInRange(Min=8, Max=20))
   > Set IdleBreak(IdleBreak=true)
   AnimNotify_Set Random Interger() {is=title K2Node_Event}
   > Set RandomInt(RandomInt:int=RandomIntegerInRange!(Min:int=0, Max:int=2) -> (ReturnValue:int=0)) -> (Output_Get:int=0)
   n5 = Event OnWeaponChanged()
   n6 = Get Dynamic Material Instance(execute=[@n5.then, @n17.then], SkeletalMeshComponent=GetPartSkeletalMeshComponent())
   > n7 = Sequence!(execute) -> (then_0, then_1)
   n8 = PickaxeItemDefinitionHasAnyMetaTag(Tags=Make Array("[0]"=(TagName="Cosmetics.MetaTags.LivelyDomino")), bRequiresEquipped=true)
   > n9 = Branch(Condition=@n8)
   > n10 = Branch(Condition=Get bIsSkyDiving())
   SetScalarParameterValue(execute=[@n12.then, @n10.then, @n11.else, @n9.else], self=@n6, ParameterName=Hide Element 01, Value=0)
   SetScalarParameterValue(execute=[@n10.else, @n11.then], self=@n6, ParameterName=Hide Element 01, Value=1)
   n11 = Branch(execute=[@n13.then, @n7.then_1], Condition=@n8)
   n12 = Event OnBeginSkydiving()
   n13 = Event OnEndSkydiving()
   n14 = Event BlueprintPostEvaluateAnimation() {is=title K2Node_Event}
   n15 = Set Active Kill(execute=[@n14.then, @n18.then], Active Kill=false)
   Branch(execute=[@n15.then, @n16.then_3], Condition=Get bShouldDisplayFrontendReactivity() -> (bShouldDisplayFrontendReactivity:bool))
   > Delay(Duration=15)
   > Set Active Kill(Active Kill=true)
   Event BlueprintInitializeAnimation()
   > n16 = Sequence!(execute) -> (then_0, then_1, then_2:Exec, then_3:Exec)
   > Delay(Duration=RandomFloatInRange(Min=0.5, Max=4))
   > Set Tail Twitch(Tail Twitch=true)
   Delay(execute=@n16.then_1, Duration=RandomFloatInRange(Min=0.5, Max=4))
   > Set Ear Wiggle(Ear Wiggle=true)
   Delay(execute=@n16.then_2, Duration=RandomFloatInRange(Min=7, Max=10))
   > Set IdleBreak(IdleBreak=true)
   n17 = Event OnCharacterCustomizationLoaded()
   n18 = AnimNotify_FE_ReactiveReset() {is=title K2Node_Event}
   TryGetPawnOwner()
   Event BlueprintUpdateAnimation()
  ]}
]
variables [
 Active Kill {type=boolean}
 As Fort Player Pawn {type=FortPlayerPawn}
 Ear Wiggle {type=boolean}
 IdleBreak {type=boolean}
 Killcount {type=number}
 RandomInt {type=integer, default=2}
 Reactive Sound Cue {type=SoundBase, default=/BRCosmetics/Sounds/Backblings/LivelyDomino/IP_Backpack_LivelyDomino_Reactive_Cue}
 Tail Twitch {type=boolean}
 TailAnimDynAlpha {type=number}
]
